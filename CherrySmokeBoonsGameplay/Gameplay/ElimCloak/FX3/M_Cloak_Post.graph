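# Post-process material for the ElimCloak boon that adds a glowing, distorted screen-edge overlay to the scene.
# Panning radial plasma noise offsets three scene samples by different amounts for chromatic aberration, then negative desaturation boosts their saturation.
# That result is multiplied by a tinted plasma glow within an inverted sphere-mask vignette.
# The MPC_Cloak_Generic_Post_MaskGlow Alpha parameter scales the effect before it is added to the unaltered scene color.
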Material /CherrySmokeBoonsGameplay/Gameplay/ElimCloak/FX3/M_Cloak_Post
graph {
 properties {materialDomain=MD_PostProcess, pixelDepthOffsetMode=PDOM_Legacy, shadingModel=MSM_Unlit, usedShadingModels=Unlit}
 nodes [
  n1 = Panner(Coordinate=Multiply(
    A=MaterialFunctionCall() {materialFunction=&VectorToRadialValue}.Radial Coordinates
    B=Constant2Vector() {g=0.2, r=10}
   )) {speedY=-0.5}
  n2 = TextureSample(UVs=@n1) {samplerType=SAMPLERTYPE_LinearGrayscale, texture=/Game/Effects/Fort_Effects/Emotes/UltraLight/Texture/T_Plasma_B_Greyscale}
  n3 = Multiply(A=AppendVector(A=@n2.R, B=@n2.R)) {constB=0.2}
  n4 = TextureCoordinate()
  n5 = SceneTexture(UVs=Add(A=@n4, B=Divide(A=@n3) {constB=6})) {sceneTextureId=PPI_PostProcessInput0}
  n6 = SceneTexture(UVs=Add(A=@n4, B=Divide(A=@n3) {constB=5})) {sceneTextureId=PPI_PostProcessInput0}
  n7 = SceneTexture(UVs=Add(A=@n4, B=Divide(A=@n3) {constB=4})) {sceneTextureId=PPI_PostProcessInput0}
  n8 = Desaturation(
   None=MaterialFunctionCall(R=ComponentMask(None=@n5.Color) {r=true}, G=ComponentMask(None=@n6.Color) {g=true}, B=ComponentMask(None=@n7.Color) {b=true}) {materialFunction=&AppendMany}.RGB
   Fraction=ScalarParameter() {defaultValue=-1}
  )
  n9 = TextureCoordinate()
  n10 = Constant2Vector() {g=0.5, r=0.5}
  n11 = Constant() {r=0.75}
  n12 = Constant()
  n13 = Power(Base=OneMinus(None=SphereMask(A=@n9, B=@n10, Radius=@n11, Hardness=@n12))) {constExponent=1.2}
  n14 = Multiply(
   A=Multiply(
    A=TextureSample(UVs=@n1) {samplerType=SAMPLERTYPE_LinearGrayscale, texture=/Game/Effects/Fort_Effects/Emotes/UltraLight/Texture/T_Plasma_C_Greyscale}.R
    B=@n13
   )
   B=Constant3Vector() {constant=<rgba 0.1486 0.8024 1 1>}
  )
  n15 = VectorParameter() {parameterName=Color, defaultValue=<rgba 0.2866 1 0.6307 1>}
  n16 = Multiply(A=Multiply(A=Add(A=@n14, B=@n15.RGB), B=ScalarParameter() {defaultValue=2, parameterName=Param_1}), B=@n13)
  Add(
   A=Multiply(
    A=Multiply(A=@n8, B=@n16)
    B=CollectionParameter() {collection=&/MPC_Cloak_Generic_Post_MaskGlow, parameterName=Alpha}
   )
   B=ComponentMask(None=SceneTexture() {sceneTextureId=PPI_PostProcessInput0}.Color) {b=true, g=true, r=true}
  )
 ]
}
